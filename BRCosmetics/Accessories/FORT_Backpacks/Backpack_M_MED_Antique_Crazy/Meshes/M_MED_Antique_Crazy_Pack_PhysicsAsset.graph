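# Physics asset for the M_MED_Antique_Crazy backpack cosmetic, providing simulation bodies for its dny_ body and left and right hand bones.
# Each of the five bones has a capsule, with a small one on the body and longer offset, angled ones on the hands.
# Constraints form two mirrored chains from body to hand base to hand.
# Every joint locks both swing axes, leaving twist as the only unlocked rotation.
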PhysicsAsset /BRCosmetics/Accessories/FORT_Backpacks/Backpack_M_MED_Antique_Crazy/Meshes/M_MED_Antique_Crazy_Pack_PhysicsAsset
bodies {
 dny_body [{shapeType=Capsule, length=0.505, radius=0.505, rotation=<rot 0 -90 0>}]
 dny_hand_base_l [{shapeType=Capsule, length=1.8, radius=2.4375, rotation=<rot 0 -90 0>}]
 dny_hand_base_r [{shapeType=Capsule, length=1.8, radius=2.4375, rotation=<rot 0 -90 0>}]
 dny_hand_l [{shapeType=Capsule, length=8.9, radius=2, center=<-0.2351 -1.5143 -4.88>, rotation=<rot 23 -90 9>}]
 dny_hand_r [{shapeType=Capsule, length=8.9, radius=2, center=<-0.2351 1.5143 4.88>, rotation=<rot 23 -90 9>}]
}
constraints [
 {
  bone1Name=dny_hand_base_r
  bone2Name=dny_body
  swing1LimitDegrees=45
  swing1Motion=Locked
  swing2LimitDegrees=45
  swing2Motion=Locked
  twistLimitDegrees=10
 }
 {
  bone1Name=dny_hand_r
  bone2Name=dny_hand_base_r
  swing1LimitDegrees=45
  swing1Motion=Locked
  swing2LimitDegrees=45
  swing2Motion=Locked
  twistLimitDegrees=10
 }
 {
  bone1Name=dny_hand_base_l
  bone2Name=dny_body
  swing1LimitDegrees=45
  swing1Motion=Locked
  swing2LimitDegrees=45
  swing2Motion=Locked
  twistLimitDegrees=10
 }
 {
  bone1Name=dny_hand_l
  bone2Name=dny_hand_base_l
  swing1LimitDegrees=45
  swing1Motion=Locked
  swing2LimitDegrees=45
  swing2Motion=Locked
  twistLimitDegrees=15
 }
]
